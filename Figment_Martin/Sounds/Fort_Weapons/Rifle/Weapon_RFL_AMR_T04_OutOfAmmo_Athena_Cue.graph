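# Sound cue for the out-of-ammo sound of a tier 4 AMR rifle in Athena.
# A mixer blends three layers: a slightly delayed, pitch-raised beam low-ammo wave with an envelope fade-out, a mechanical low-ammo sound, and a slowdown-field enter sound pitched up by a modulator.
# It uses small-rifle attenuation and the first-person ranged fire sound class, with overridden concurrency that limits instances per owner and stops the oldest.
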SoundCue /Figment_Martin/Sounds/Fort_Weapons/Rifle/Weapon_RFL_AMR_T04_OutOfAmmo_Athena_Cue
root=Mixer_0
nodes [
 {id=Delay_0, children [Enveloper_0], properties {delayMax=0.1, delayMin=0.1}}
 {
  id=Enveloper_0
  children [WavePlayer_2]
  properties {pitchCurve {editorCurveData {keys [<at 0 1.3222>]}}, volumeCurve {editorCurveData {keys [<at 0 1>,<at 0.5 1>,<at 0.6548 -0.2668>]}}}
 }
 {id=Mixer_0, children [Delay_0,WavePlayer_3,Modulator_1], properties {inputVolume [0.4,1,1]}}
 {id=Modulator_1, children [WavePlayer_5], properties {pitchMax=2, pitchMin=2}}
 {id=WavePlayer_2, properties {soundWaveAssetPtr=/Game/Sounds/Fort_Weapons/AMR/AMR_Beam_LowAmmo_01}}
 {id=WavePlayer_3, properties {soundWaveAssetPtr=/Game/Sounds/Fort_Weapons/LowAmmo/mechanical_lowammo_01}}
 {id=WavePlayer_5, properties {soundWaveAssetPtr=/Game/Sounds/Fort_Gadgets/SlowdownField/SDZ_EnemyEnter_001}}
]
properties {
 attenuationSettings=/Game/Sounds/Attenuation/Fort_Weapon_Rifle_Sml_Att
 bOverrideConcurrency=true
 maxDistance=3000
 soundClassObject=&Fort_WeaponRanged_Fire_1P_PSM
 volumeMultiplier=1.69
 attenuationOverrides {distanceAlgorithm=NaturalSound, falloffDistance=4000, attenuationShapeExtents {x=1000}}
 concurrencyOverrides {bLimitToOwner=true, maxCount=2, resolutionRule=StopOldest}
}
